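# Skewed frontend button widget built on CommonButtonLegacy, with input action icon, rich text and optional right-side image over a faint shadow copy, defaulting to Battle Pass art.
# PreConstruct applies exposed text, font size, shear, color, justification, wrapping, background material and texture, collapsing the image when none is set.
# Hovering plays one animation forward and unhovering reverses it: SimpleHover selects a scale-up instead of animated material parameters with shifting content.
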WidgetBlueprint /FortniteCoreUI/Frontend/Spatial/Customization/Spatial_Specialbutton
animations [
 Hover {
  binding_count=3
  length_seconds=0.1
  object=&self:Hover
  sequence {bindings [
    Border_Material {
     id=#1
     possessed_class=Border
     tracks [{
       type=MovieSceneWidgetMaterialTrack
       sections [{
         type=MovieSceneParameterSection
         channels [
          Buttonn_Opcaity {type=MovieSceneScriptingFloatChannel, keys [[0,0.8],[2,1]]}
          Reflection_Opcaity {type=MovieSceneScriptingFloatChannel, keys [[0,0.125],[2,0.2]]}
          Reflection anim {type=MovieSceneScriptingFloatChannel, keys [[0,0],[2,0.5]]}
         ]
        }]
      }]
    }
    Image_45 {
     id=#2
     possessed_class=Image
     tracks [{
       type=MovieScene2DTransformTrack
       sections [{
         type=MovieScene2DTransformSection
         end_frame=2
         start_frame=0
         channels [
          Translation.X {type=MovieSceneScriptingFloatChannel, keys [[0,0],[2,-3]]}
          Translation.Y {type=MovieSceneScriptingFloatChannel, keys [[0,0],[2,-6]]}
          Angle {type=MovieSceneScriptingFloatChannel}
          Scale.X {type=MovieSceneScriptingFloatChannel}
          Scale.Y {type=MovieSceneScriptingFloatChannel}
          Shear.X {type=MovieSceneScriptingFloatChannel}
          Shear.Y {type=MovieSceneScriptingFloatChannel}
         ]
        }]
      }]
    }
    HorizontalBox_31 {
     id=#3
     possessed_class=HorizontalBox
     tracks [{
       type=MovieScene2DTransformTrack
       sections [{
         type=MovieScene2DTransformSection
         end_frame=2
         start_frame=0
         channels [
          Translation.X {type=MovieSceneScriptingFloatChannel, keys [[0,0],[2,8]]}
          Translation.Y {type=MovieSceneScriptingFloatChannel, keys [[0,0],[2,8]]}
          Angle {type=MovieSceneScriptingFloatChannel}
          Scale.X {type=MovieSceneScriptingFloatChannel}
          Scale.Y {type=MovieSceneScriptingFloatChannel}
          Shear.X {type=MovieSceneScriptingFloatChannel}
          Shear.Y {type=MovieSceneScriptingFloatChannel}
         ]
        }]
      }]
    }
   ]}
 }
 Hover2 {
  binding_count=1
  length_seconds=0.1
  object=&self:Hover2
  sequence {bindings [Border_Material {
     id=#4
     possessed_class=Border
     tracks [{
       type=MovieScene2DTransformTrack
       sections [{
         type=MovieScene2DTransformSection
         channels [
          Scale.X {type=MovieSceneScriptingFloatChannel, keys [[0,1],[2,1.05]]}
          Scale.Y {type=MovieSceneScriptingFloatChannel, keys [[0,1],[2,1.05]]}
         ]
        }]
      }]
    }]}
 }
]
graphs [EventGraph {nodes [
   Event On Hovered()
   n1 = Select(
    Option 0:Widget Animation Object Reference=Get Hover() -> (Hover:Widget Animation Object Reference)
    Option 1:Widget Animation Object Reference=Get Hover2() -> (Hover2:Widget Animation Object Reference)
    Index=Get SimpleHover() -> (SimpleHover:bool)
   ) -> (ReturnValue:Widget Animation Object Reference)
   > PlayAnimationForward(InAnimation=@n1)
   Event On Unhovered()
   > PlayAnimationReverse(InAnimation=@n1)
   Event PreConstruct()
   n2 = Select(Option 0:int=34, Option 1:int=Get FontSize() -> (FontSize:int=0), Index=Get FontSizeOverride() -> (FontSizeOverride:bool)) -> (ReturnValue:int)
   n3 = Get Rich_text() -> (Rich_text:Common Rich Text Block Object Reference)
   > SetDefaultFont!(
    execute
    self:Rich Text Block Object Reference=@n3
    InFontInfo:Slate Font Info Structure=MakeSlateFontInfo!(
     FontObject:Object Reference=&Burbank
     FontMaterial:Object Reference
     OutlineSettings:Font Outline Settings Structure=(OutlineSize=0,bMiteredCorners=False,bSeparateFillAlpha=False,bApplyOutlineToDropShadows=False,OutlineMaterial=None,OutlineColor=(R=0,G=0,B=0,A=1))
     TypefaceFontName:Name=BurbankRegular-black
     Size:float="To Float (Integer)"(InInt=@n2)
     LetterSpacing:int=0
     SkewAmount:float=0
     bForceMonospaced:bool=false
     bMaterialIsStencil:bool=false
     MonospacedWidth:float=1
    ) -> (ReturnValue:Slate Font Info Structure)
   ) -> (then)
   > SetRenderShear!(
    execute
    self:Widget Object Reference=@n3
    Shear:Vector 2D Structure=Select(Option 0:Vector 2D Structure, Option 1:Vector 2D Structure=(X=-12,Y=0), Index=Get ShearText() -> (ShearText:bool)) -> (ReturnValue:Vector 2D Structure)
   ) -> (then)
   > SetText(self=@n3, InText=Get Button text() -> (Button text:Text))
   > SetDefaultColorAndOpacity!(
    execute
    self:Rich Text Block Object Reference=@n3
    InColorAndOpacity:Slate Color Structure=Get Text Color() -> (Text Color:Slate Color Structure)
   ) -> (then)
   > Set Justification!(
    execute
    Justification:ETextJustify Enum=Get Justifiction() -> (Justifiction:ETextJustify Enum=Left)
    self:Text Layout Widget Object Reference=@n3
   ) -> (then, Output_Get:ETextJustify Enum=Left)
   > SetAutoWrapText!(execute, self:Rich Text Block Object Reference=Get Rich_text(), InAutoTextWrap:bool=Get Auto wrap text() -> (Auto wrap text:bool)) -> (then)
   > SetBrushFromMaterial(
    self:Border Object Reference=Get Border_Material() -> (Border_Material:Border Object Reference)
    Material=Get Material Brush() -> (Material Brush:Material Interface Object Reference)
   )
   n4 = Get Right Image() -> (Right Image:Texture 2D Object Reference)
   > SetVisibility(
    self=Get Overlay_Right_Image() -> (Overlay_Right_Image:Overlay Object Reference)
    InVisibility=Select(Option 0:ESlateVisibility Enum=Collapsed, Option 1:ESlateVisibility Enum=SelfHitTestInvisible, Index=IsValid(Object=@n4)) -> (ReturnValue:ESlateVisibility Enum)
   )
   > SetBrushFromTexture!(
    execute
    self:Image Object Reference=[Get Image() -> (Image:Image Object Reference), Get Image_45() -> (Image_45:Image Object Reference)]
    Texture:Texture 2D Object Reference=Get Right Image()
    bMatchSize:bool=false
   ) -> (then)
   Set SimpleHover(SimpleHover:bool=false) -> (Output_Get:bool)
  ]}]
variables [
 Auto wrap text {type=boolean, default=true}
 Button text {type=string, default=Button text}
 FontSize {type=integer}
 FontSizeOverride {type=boolean}
 Justifiction {type=ETextJustify, default=Left}
 Material Brush {type=MaterialInterface, default=/Game/Athena/UI/Frontend/Spatial/Art/M_Ui_Skewed_SpatialButton_Default}
 Right Image {type=Texture2D, default=/Game/UI/Foundation/Textures/Icons/Items/T-FNBR-BattlePass}
 ShearText {type=boolean}
 SimpleHover {type=boolean}
 Text Color {type=SlateColor, default {colorUseRule=UseColor_Specified, specifiedColor=<rgba 1 1 1 1>}}
]
widgets {
 info {inheritedWidgetCount=1, parentClass=&CommonButtonLegacy, rootWidgetClass=&Border, widgetCount=9}
 widgets [
  Border_Material {
   widgetClassPath=&Border
   bIsVariable=true
   widget=Border_Material
   properties {
    background {
     bIsDynamicallyLoaded=false
     drawAs=Box
     imageType=NoImage
     mirroring=NoMirror
     resourceName=None
     resourceObject=&M_Ui_Skewed_SpatialButton_Default
     tiling=NoTile
     imageSize=<128 128>
     margin=<ltrb 0.45 0 0.45 0>
     outlineSettings {
      bUseBrushTransparency=false
      roundingType=HalfHeightRadius
      width=0
      color {colorUseRule=UseColor_Specified, specifiedColor=<rgba 0 0 0 0>}
      cornerRadii=<0 0 0 0>
     }
     tintColor {colorUseRule=UseColor_Specified, specifiedColor=<rgba 1 1 1 1>}
     uVRegion {bIsValid=false, max=<0 0>, min=<0 0>}
    }
    padding=<ltrb 48 0 48 0>
   }
  }
  SizeBox_Height {
   widgetClassPath=&SizeBox
   bIsVariable=true
   parent=Border_Material
   widget=SizeBox_Height
   properties {bOverride_MinDesiredHeight=true, minDesiredHeight=150}
   slot {class=BorderSlot, properties {padding=<ltrb 48 0 48 0>}}
  }
  HorizontalBox_31 {widgetClassPath=&HorizontalBox, parent=SizeBox_Height, widget=HorizontalBox_31, slot {class=SizeBoxSlot}}
  HorizontalBox_0 {
   widgetClassPath=&HorizontalBox
   parent=HorizontalBox_31
   widget=HorizontalBox_0
   slot {
    class=HorizontalBoxSlot
    properties {horizontalAlignment=HAlign_Center, verticalAlignment=VAlign_Center, padding=<ltrb 0 0 0 4>, size {sizeRule=Fill, value=1}}
   }
  }
  InputActionWidget {
   widgetClassPath=&CommonActionWidget
   bInherited=true
   bIsVariable=true
   parent=HorizontalBox_0
   widget=InputActionWidget
   properties {renderTransform {angle=0, scale=<1.25 1.25>, shear=<0 0>, translation=<0 0>}}
   slot {class=HorizontalBoxSlot, properties {padding=<ltrb 0 0 24 0>}}
  }
  Rich_text {
   widgetClassPath=&CommonRichTextBlock
   bIsVariable=true
   parent=HorizontalBox_0
   widget=Rich_text
   properties {
    bOverrideDefaultStyle=true
    bTintInlineIcon=true
    justification=Center
    lineHeightPercentage=0.8
    text="Purchase\r\n<text color=\"FFFFFFFF\">Battle Pass!</>"
    textTransformPolicy=ToUpper
    defaultTextStyleOverride {
     overflowPolicy=Clip
     transformPolicy=None
     colorAndOpacity {colorUseRule=UseColor_Specified, specifiedColor=<rgba 1 1 1 1>}
     font {
      bForceMonospaced=false
      bMaterialIsStencil=false
      fontMaterial=null
      fontObject=&Burbank
      letterSpacing=0
      monospacedWidth=1
      size=34
      skewAmount=0
      typefaceFontName=Burbank-Black
      outlineSettings {
       bApplyOutlineToDropShadows=false
       bMiteredCorners=false
       bSeparateFillAlpha=false
       outlineMaterial=null
       outlineSize=0
       outlineColor=<rgba 0 0 0 1>
      }
     }
     highlightColor {colorUseRule=UseColor_Specified, specifiedColor=<rgba 0 0 0 1>}
     highlightShape {
      bIsDynamicallyLoaded=false
      drawAs=Image
      imageType=NoImage
      mirroring=NoMirror
      resourceName=None
      resourceObject=null
      tiling=NoTile
      imageSize=<32 32>
      margin=<ltrb 0 0 0 0>
      outlineSettings {
       bUseBrushTransparency=false
       roundingType=HalfHeightRadius
       width=0
       color {colorUseRule=UseColor_Specified, specifiedColor=<rgba 0 0 0 0>}
       cornerRadii=<0 0 0 0>
      }
      tintColor {colorUseRule=UseColor_Specified, specifiedColor=<rgba 1 1 1 1>}
      uVRegion {bIsValid=false, max=<0 0>, min=<0 0>}
     }
     selectedBackgroundColor {colorUseRule=UseColor_Foreground, specifiedColor=<rgba 1 0 1 1>}
     shadowColorAndOpacity=<rgba 0 0 0 1>
     shadowOffset=<0 0>
     strikeBrush {
      bIsDynamicallyLoaded=false
      drawAs=Image
      imageType=NoImage
      mirroring=NoMirror
      resourceName=None
      resourceObject=null
      tiling=NoTile
      imageSize=<32 32>
      margin=<ltrb 0 0 0 0>
      outlineSettings {
       bUseBrushTransparency=false
       roundingType=HalfHeightRadius
       width=0
       color {colorUseRule=UseColor_Specified, specifiedColor=<rgba 0 0 0 0>}
       cornerRadii=<0 0 0 0>
      }
      tintColor {colorUseRule=UseColor_Specified, specifiedColor=<rgba 1 1 1 1>}
      uVRegion {bIsValid=false, max=<0 0>, min=<0 0>}
     }
     underlineBrush {
      bIsDynamicallyLoaded=false
      drawAs=Image
      imageType=NoImage
      mirroring=NoMirror
      resourceName=None
      resourceObject=null
      tiling=NoTile
      imageSize=<32 32>
      margin=<ltrb 0 0 0 0>
      outlineSettings {
       bUseBrushTransparency=false
       roundingType=HalfHeightRadius
       width=0
       color {colorUseRule=UseColor_Specified, specifiedColor=<rgba 0 0 0 0>}
       cornerRadii=<0 0 0 0>
      }
      tintColor {colorUseRule=UseColor_Specified, specifiedColor=<rgba 1 1 1 1>}
      uVRegion {bIsValid=false, max=<0 0>, min=<0 0>}
     }
    }
    margin=<ltrb 0 0 0 10>
   }
   slot {class=HorizontalBoxSlot, properties {verticalAlignment=VAlign_Center, size {sizeRule=Fill, value=1}}}
  }
  Overlay_Right_Image {
   widgetClassPath=&Overlay
   bIsVariable=true
   parent=HorizontalBox_31
   widget=Overlay_Right_Image
   slot {class=HorizontalBoxSlot, properties {verticalAlignment=VAlign_Center, padding=<ltrb 16 0 0 0>}}
  }
  Image {
   widgetClassPath=&Image
   bIsVariable=true
   parent=Overlay_Right_Image
   widget=Image
   properties {
    visibility=SelfHitTestInvisible
    brush {
     bIsDynamicallyLoaded=false
     drawAs=Image
     imageType=NoImage
     mirroring=NoMirror
     resourceName=None
     resourceObject=&T-FNBR-BattlePass
     tiling=NoTile
     imageSize=<128 128>
     margin=<ltrb 0 0 0 0>
     outlineSettings {
      bUseBrushTransparency=false
      roundingType=HalfHeightRadius
      width=0
      color {colorUseRule=UseColor_Specified, specifiedColor=<rgba 0 0 0 0>}
      cornerRadii=<0 0 0 0>
     }
     tintColor {colorUseRule=UseColor_Specified, specifiedColor=<rgba 0 0 0 1>}
     uVRegion {bIsValid=false, max=<0 0>, min=<0 0>}
    }
    colorAndOpacity=<rgba 1 1 1 0.112>
    renderTransform {angle=0, scale=<1 1>, shear=<0 0>, translation=<6 6>}
   }
   slot {class=OverlaySlot}
  }
  Image_45 {
   widgetClassPath=&Image
   bIsVariable=true
   parent=Overlay_Right_Image
   widget=Image_45
   properties {
    visibility=SelfHitTestInvisible
    brush {
     bIsDynamicallyLoaded=false
     drawAs=Image
     imageType=NoImage
     mirroring=NoMirror
     resourceName=None
     resourceObject=&T-FNBR-BattlePass
     tiling=NoTile
     imageSize=<128 128>
     margin=<ltrb 0 0 0 0>
     outlineSettings {
      bUseBrushTransparency=false
      roundingType=HalfHeightRadius
      width=0
      color {colorUseRule=UseColor_Specified, specifiedColor=<rgba 0 0 0 0>}
      cornerRadii=<0 0 0 0>
     }
     tintColor {colorUseRule=UseColor_Specified, specifiedColor=<rgba 1 1 1 1>}
     uVRegion {bIsValid=false, max=<0 0>, min=<0 0>}
    }
   }
   slot {class=OverlaySlot}
  }
 ]
}
